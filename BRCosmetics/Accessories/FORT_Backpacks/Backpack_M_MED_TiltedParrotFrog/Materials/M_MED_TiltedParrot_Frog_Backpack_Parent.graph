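# Parent character material for the TiltedParrot Frog backpack, recoloring it and adding glow as eliminations increase, behind an EnableReactivity switch.
# Timed, eased blends between previous and current elimination counts drive an eight-division progress value that samples curve-atlas rows for region colors and emissive scale.
# SkinFX mask channels, including sphere-masked gradient bands, place those colors between the shared character PreFX/SkinFX and PostFX functions.
# A debug switch cycles stages over time.
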Material /BRCosmetics/Accessories/FORT_Backpacks/Backpack_M_MED_TiltedParrotFrog/Materials/M_MED_TiltedParrot_Frog_Backpack_Parent
graph {
 properties {
  bAutomaticallySetUsageInEditor=false
  bUseMaterialAttributes=true
  bUsedWithMeshDeformer=true
  bUsedWithMorphTargets=true
  bUsedWithSkeletalMesh=true
  pixelDepthOffsetMode=PDOM_Legacy
  refractionMethod=RM_IndexOfRefraction
  shadingModel=MSM_FromMaterialExpression
  subsurfaceProfile=&SS_HeroSkin_02
  usedShadingModels=Default Lit | Subsurface | Clear Coat | Subsurface Profile | Cloth | Eye
 }
 nodes [
  n1 = MaterialFunctionCall() {materialFunction=&BaseFN_CharacterFunction_PreFX}
  n2 = MaterialFunctionCall(
   PreskinnedPosition=@n1.Pre skinned Local Position
   Gmap=@n1.Pre skinned Local Position
   MaterialAttributes=@n1.Pre skinned Local Position
   AO=@n1.Pre skinned Local Position
  ) {materialFunction=&CharacterSkinFX}
  n3 = GetMaterialAttributes(None=@n2.Result) {attributeGetTypes [#1, #2]}
  n4 = LinearInterpolate(A=NamedRerouteUsage(), B=NamedRerouteUsage(), Alpha=NamedRerouteUsage())
  n5 = NamedRerouteUsage()
  n6 = NamedRerouteUsage()
  n7 = LinearInterpolate(
   A=LinearInterpolate(A=LinearInterpolate(A=@n4, B=@n5, Alpha=@n6), B=NamedRerouteUsage(), Alpha=NamedRerouteUsage())
   B=NamedRerouteUsage()
   Alpha=NamedRerouteUsage()
  )
  n8 = MaterialFunctionCall(Float3=@n2.SkinFX_Mask) {materialFunction=&BreakOutFloat3Components}
  n9 = LinearInterpolate(A=@n3.Base Color, B=@n7, Alpha=@n8.G)
  n10 = ScalarParameter() {defaultValue=7, group=Reactivity, parameterName=ReactiveEmissiveAmt}
  n11 = OneMinus(None=Fresnel(ExponentIn=ScalarParameter() {defaultValue=0.23, group=Reactivity, parameterName=ReactiveEmissiveExp}) {baseReflectFraction=0})
  n12 = NamedRerouteUsage()
  n13 = Multiply(
   A=NamedRerouteUsage()
   B=LinearInterpolate(
    A=Multiply(A=@n10, B=@n11)
    B=VectorParameter() {group=Reactivity, parameterName=Reactive_FinalStageSpotsEmissive, defaultValue=<rgba 0 3 0.25 1>}.RGB
    Alpha=Saturate(None=@n12)
   )
  )
  n14 = Add(A=@n3.Emissive Color, B=Multiply(A=@n8.G, B=Multiply(A=@n7, B=@n13)))
  n15 = MaterialFunctionCall(
   AO=@n1.AO
   Distance Blend=@n1.AO
   Fresnel=@n1.AO
   Gmap=@n1.AO
   Pre Skinned Local Position=@n1.AO
   Material=StaticSwitchParameter(
    True=SetMaterialAttributes!(MaterialAttributes=@n3.MaterialAttributes, Base Color=@n9, Emissive Color=@n14) {attributeSetTypes [#1, #2]}
    False=@n3.MaterialAttributes
   ) {defaultValue=true, group=Reactivity, parameterName=EnableReactivity, sortPriority=0}
  ) {materialFunction=&BaseFN_CharacterFunction_PostFX}
  MaterialFunctionCall(In=@n15.Result) {materialFunction=&CINE_MF_CharPostFX}
  n16 = Time()
  n17 = ScalarParameter() {defaultValue=3, group=Reactivity, parameterName=ReactiveDuration}
  n18 = Divide(A=Subtract(A=Time(), B=ScalarParameter() {defaultValue=-100, group=Reactivity, parameterName=ReactiveStartTime_CCPM}), B=@n17)
  n19 = StaticSwitchParameter(
   True=Fmod(A=Divide(A=@n16, B=@n17), B=Constant() {r=8})
   False=LinearInterpolate(
    A=ScalarParameter() {group=Reactivity, parameterName=PreviousEliminations_CCPM}
    B=ScalarParameter() {group=Reactivity, parameterName=Eliminations_CCPM}
    Alpha=Saturate(None=@n18)
   )
  ) {group=Reactivity, parameterName=DEBUG REACTIVITY}
  n20 = Floor(None=@n19)
  n21 = Add(
   A=MaterialFunctionCall!(
    Input=Subtract(A=@n19, B=@n20)
    InputBias
    CurveGain=ScalarParameter() {defaultValue=4, group=Reactivity, parameterName=ReactiveTransitionSteepness}
    CurveBias
   ) {materialFunction=/Game/Global/MaterialFunctions/MF_PolynomialEaseInOut}.Gain + Bias
   B=@n20
  )
  NamedRerouteDeclaration(None=Saturate(None=Divide(A=@n21) {constB=8, desc="Color curves of this skin have 8 curve divisions rather than 7, to better match the pixel grid"})) {name=ReactivityProgress, nodeColor=<rgba 0.0781 1 0 1>}
  n22 = CurveAtlasRowParameter(CurveTime=NamedRerouteUsage()) {
   atlas=&/CA_M_MED_TiltedParrot_Frog_Backpack
   curve=&/C_M_MED_TiltedParrot_Frog_Backpack_A
   desc=Eyes
   group=Reactivity
   parameterName=ReactiveColor_A
  }
  NamedRerouteDeclaration(None=@n22) {name=Color_Eyes, nodeColor=<rgba 0.8516 0 1 1>}
  NamedRerouteDeclaration(None=SphereMask(
    A=NamedRerouteUsage()
    B=Constant() {r=1}
    Radius=ScalarParameter() {defaultValue=0.09, group=Reactivity, parameterName=Reactive_EyesRadius, sliderMax=0.2}
    Hardness=ScalarParameter() {defaultValue=0.8, group=Reactivity, parameterName=Reactive_EyesHardness, sliderMax=1}
   ) {attenuationRadius=0.15, hardnessPercent=0.9}) {name=Mask_Eyes, nodeColor=<rgba 0.2656 1 0 1>}
  NamedRerouteDeclaration(None=SphereMask(
    A=NamedRerouteUsage()
    B=Constant()
    Radius=ScalarParameter() {defaultValue=0.79, group=Reactivity, parameterName=Reactive_LimbsRadius, sliderMax=0.2}
    Hardness=ScalarParameter() {group=Reactivity, parameterName=Reactive_LimbsHardness, sliderMax=1}
   ) {attenuationRadius=0.15, hardnessPercent=0.9}) {name=Mask_Limbs, nodeColor=<rgba 0.2656 1 0 1>}
  n23 = TextureSampleParameter2D() {
   desc=Separate sampler node here because SkinFXMask output above doesn't include alpha channel
   parameterName=SkinFX_Mask
   samplerSource=SSM_Wrap_WorldGroupSettings
   samplerType=SAMPLERTYPE_Masks
   texture=/BRCosmetics/Accessories/FORT_Backpacks/Backpack_M_MED_TiltedParrotFrog/Textures/T_M_MED_TiltedParrotFrog_Backpack_FX
  }
  NamedRerouteDeclaration(None=@n23.A) {name=Mask_Spots, nodeColor=<rgba 0.2656 1 0 1>}
  n24 = CurveAtlasRowParameter(CurveTime=NamedRerouteUsage()) {
   atlas=&CA_M_MED_TiltedParrot_Frog_Backpack
   curve=&/C_M_MED_TiltedParrot_Frog_Backpack_B
   defaultValue=1
   desc=Secondary
   group=Reactivity
   parameterName=ReactiveColor_B
  }
  NamedRerouteDeclaration(None=@n24) {name=Color_Secondary, nodeColor=<rgba 0 1 0.6484 1>}
  n25 = CurveAtlasRowParameter(CurveTime=NamedRerouteUsage()) {
   atlas=&CA_M_MED_TiltedParrot_Frog_Backpack
   curve=&/C_M_MED_TiltedParrot_Frog_Backpack_C
   defaultValue=2
   desc=Primary
   group=Reactivity
   parameterName=ReactiveColor_C
  }
  NamedRerouteDeclaration(None=@n25) {name=Color_Primary, nodeColor=<rgba 0.0547 1 0 1>}
  n26 = CurveAtlasRowParameter(CurveTime=NamedRerouteUsage()) {
   atlas=&CA_M_MED_TiltedParrot_Frog_Backpack
   curve=&/C_M_MED_TiltedParrot_Frog_Backpack_D
   defaultValue=3
   desc=Limbs
   group=Reactivity
   parameterName=ReactiveColor_D
  }
  NamedRerouteDeclaration(None=@n26) {name=Color_Limbs, nodeColor=<rgba 0 1 0.0859 1>}
  n27 = CurveAtlasRowParameter(CurveTime=NamedRerouteUsage()) {
   atlas=&CA_M_MED_TiltedParrot_Frog_Backpack
   curve=&/C_M_MED_TiltedParrot_Frog_Backpack_E
   defaultValue=4
   desc=Spots
   group=Reactivity
   parameterName=ReactiveColor_E
  }
  NamedRerouteDeclaration(None=@n27) {name=Color_Spots, nodeColor=<rgba 0.6172 0 1 1>}
  n28 = CurveAtlasRowParameter(CurveTime=NamedRerouteUsage()) {
   atlas=&CA_M_MED_TiltedParrot_Frog_Backpack
   curve=&/C_M_MED_TiltedParrot_Frog_Backpack_F
   defaultValue=5
   desc="R = Emissive on/off"
   group=Reactivity
   parameterName=ReactiveColor_EmissiveScale
  }
  NamedRerouteDeclaration(None=@n28.R) {name=EmissiveScale, nodeColor=<rgba 0.9219 0 1 1>}
  NamedRerouteDeclaration(None=SphereMask(
    A=NamedRerouteUsage()
    B=Constant() {r=0.9}
    Radius=ScalarParameter() {defaultValue=0.09, group=Reactivity, parameterName=Reactive_SecondaryRadius, sliderMax=0.2}
    Hardness=ScalarParameter() {defaultValue=0.5, group=Reactivity, parameterName=Reactive_SecondaryHardness, sliderMax=1}
   ) {attenuationRadius=0.15, hardnessPercent=0.9}) {name=Mask_Secondary, nodeColor=<rgba 0.2656 1 0 1>}
  NamedRerouteDeclaration(None=@n23.B) {name=Color Map Raw, nodeColor=<rgba 0.4531 0 1 1>}
 ]
}
